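# Foundations_Ambience_9x9_PSM is a SoundClass in FigmentCore, a child of the Foundations_Ambience class.
# It attaches the Foundations_DuckAmbience_9x9 sound mix as a passive modifier with a zero minimum volume threshold, so that mix applies automatically while sounds of this class play.
# It also sets its own volume and LFE bleed.
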SoundClass /FigmentCore/Sounds/FortSoundClassesAndModes/Foundations_Ambience_9x9_PSM
parent=&/Foundations_Ambience
properties {
 passiveSoundMixModifiers [{minVolumeThreshold=0, soundMix=&/Foundations_DuckAmbience_9x9}]
 properties {lFEBleed=0.2, volume=1.3}
}
