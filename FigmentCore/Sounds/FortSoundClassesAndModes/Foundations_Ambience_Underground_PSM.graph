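# Sound class for underground ambience, set up as a root class with no parent.
# It routes audio to the Fortnite_SFX submix by default and lets ambient volumes affect its sounds.
# Volume is modulated through the CB_FortAmbient control bus from the UEFN modulation buses, and its sounds prime on load.
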SoundClass /FigmentCore/Sounds/FortSoundClassesAndModes/Foundations_Ambience_Underground_PSM
parent=null
properties {properties {
  bApplyAmbientVolumes=true
  defaultSubmix=&Fortnite_SFX
  loadingBehavior=PrimeOnLoad
  modulationSettings {volumeModulationDestination {bEnableModulation=true, value=1, modulators [/Game/Sounds/Modulation/Buses/UEFN/CB_FortAmbient]}}
 }}
