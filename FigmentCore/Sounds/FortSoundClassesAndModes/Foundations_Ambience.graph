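# Foundations_Ambience is a Sound Class that groups ambience audio under the Fort_12_Base_Ambient parent class.
# It branches into four child classes named for InBuilding, OpenSpace, 9x9 and Underground ambience, each with a PSM suffix.
# The class applies effects, bleeds slightly into the LFE channel and raises volume above unity.
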SoundClass /FigmentCore/Sounds/FortSoundClassesAndModes/Foundations_Ambience
parent=/Game/Sounds/FortSoundClassesAndModes/Fort_12_Base_Ambient
children [
 &/Foundations_Ambience_InBuilding_PSM
 &/Foundations_Ambience_OpenSpace_PSM
 &/Foundations_Ambience_9x9_PSM
 &/Foundations_Ambience_Underground_PSM
]
properties {properties {bApplyEffects=true, lFEBleed=0.1, volume=1.2}}
